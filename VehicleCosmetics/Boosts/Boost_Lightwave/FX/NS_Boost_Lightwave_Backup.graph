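# Niagara boost trail effect for the Lightwave vehicle cosmetic, named as a backup, emitting rearward ribbons and sprites from skeletal mesh sockets.
# Ribbons spawn per port at quality-based rates, stack five renderers and taper, using painted-variant colors or an orange-to-blue default gradient.
# Sprites are split across sockets and swirl through vortex and curl noise forces.
# Both inherit source movement; alternate Ribbons001 and Sprites002 emitters are disabled.
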NiagaraSystem /VehicleCosmetics/Boosts/Boost_Lightwave/FX/NS_Boost_Lightwave_Backup
dependencies {
 usedDataInterfaces [
  &NiagaraDataInterfaceCurve
  &NiagaraDataInterfaceVectorCurve
  &NiagaraDataInterfaceColorCurve
  &NiagaraDataInterfaceArrayColor
  &NiagaraDataInterfaceSkeletalMesh
 ]
 usedDynamicInputs []
 usedModules [
  &self:SystemSpawnScript.SystemScriptSource.SystemScriptGraph.NiagaraNodeAssignment_0.SetVariables_#1
  &SystemState
 ]
}
emitters {
 Ribbons {
  emitterUpdateScript [
   &EmitterState(Inactive Response=::NewEnumerator1, Life Cycle Mode=::NewEnumerator1, Loop Duration=1, UseLoopDelay=false)
   &VC_FilteredSocketCount()
   &VC_RateQualityHandler(Epic Spawn Rate=50, High Spawn Rate=40, Low Spawn Rate=30, Use Custom Rate?=true)
   &VC_SpawnRibbon_v1(Number of Ribbons=$Emitter.PortCount, SpawnRate=$Emitter.Spawn Rate)
  ]
  particleSpawnScript [
   &VC_PaintedVariantsOverride(
    Black=<rgba 0.027 0.027 0.027 1>
    BurntSienna=<rgba 0.077 0.037 0.0212 1>
    Cobalt=<rgba 0 0 0.15 1>
    Crimson=<rgba 0.13 0 0.0217 1>
    ForestGreen=<rgba 0.003 0.1 0.0272 1>
    Grey=<rgba 0.062 0.062 0.062 1>
    Lime=<rgba 0.0675 0.075 0.03 1>
    Orange=<rgba 0.112 0.0481 0.0207 1>
    Pink=<rgba 0.105 0.0315 0.0866 1>
    Purple=<rgba 0.07 0 0.12 1>
    R1 Black:LinearColor=<rgba 0.027 0.027 0.027 1>
    R1 BurntSienna:LinearColor=<rgba 0.077 0.037 0.0212 1>
    R1 Cobalt:LinearColor=<rgba 0 0 0.15 1>
    R1 Crimson:LinearColor=<rgba 0.13 0 0.0217 1>
    R1 ForestGreen:LinearColor=<rgba 0.003 0.1 0.0272 1>
    R1 Grey:LinearColor=<rgba 0.062 0.062 0.062 1>
    R1 Lime:LinearColor=<rgba 0.0675 0.075 0.03 1>
    R1 Orange:LinearColor=<rgba 0.112 0.0481 0.0207 1>
    R1 Pink:LinearColor=<rgba 0.105 0.0315 0.0866 1>
    R1 Purple:LinearColor=<rgba 0.071 0 0.1217 0.087>
    R1 Saffron:LinearColor=<rgba 0.085 0.0786 0.0085 1>
    R1 SkyBlue:LinearColor=<rgba 0.0165 0.0679 0.11 1>
    R1 TitaniumWhite:LinearColor=<rgba 0.119 0.119 0.14 1>
    Saffron=<rgba 0.085 0.0786 0.0085 1>
    SkyBlue=<rgba 0.0165 0.0679 0.11 1>
    TitaniumWhite=<rgba 0.119 0.119 0.14 1>
    Use R1 Painted Overrides=true
   )
   &VC_InitializeRibbon_v1(
    Color=$Particles.Color
    Custom Offset=true
    Enable Skeletal Position=true
    Lifetime=0.8
    Local Vector Offset=&AddVectorToPosition(
     Position=/Niagara/DynamicInputs/TypeConversions/MakeCustomPositionFromBool(Bool:NiagaraBool=$User.bIsFrontendPreview, True Position:NiagaraPosition=<-10 0 0>, False Position:NiagaraPosition=<0 0 0>)
     Vector=$User.DTCarOffset
    )
    Offset Amount=1
    Ribbon Width=120
   )
   &AddVelocity(Velocity=<-2500 0 0>) {module=AddVelocity001}
  ]
  particleUpdateScript [
   &ParticleState()
   &InheritSourceMovement(
    Applied Position Scale=&Multiply_VectorByFloat(
     Vector=<1 1 1>
     Float=&FloatFromCurve(FloatCurve=DataInterface() {propertyValues {Curve {keys [
          {interpMode=RCIM_Cubic, leaveTangent=-0.0051, tangentMode=RCTM_Break, time=0, value=1}
          {arriveTangent=-0.247, interpMode=RCIM_Cubic, tangentMode=RCTM_Break, time=1, value=0.9}
         ]}}})
    )
    ApplyForce=false
    ApplyPosition=true
    ApplyPositionFalloff=false
   )
   &SolveForcesAndVelocity(Clamp Velocity=false, Limit Acceleration=false)
   &VC_UpdateRibbon_v1(Enable Skeletal Position=true)
   &Color(Color=&MakeCustomLinearColorFromBool(
     Bool=$User.bIsPainted
     True Color=$Particles.Color
     False Color=&ColorFromCurve(ColorCurve=DataInterface() {propertyValues {
        AlphaCurve {keys [<at 0 1 RCIM_Cubic>]}
        BlueCurve {keys [<at 0.1166 0.045>,<at 0.6508 0.9>]}
        GreenCurve {keys [<at 0.1166 0.4725>,<at 0.6508 0.225>]}
        RedCurve {keys [<at 0.1166 0.9>,<at 0.6508 0.09>]}
       }})
    ))
   /SetVariables(
    Particles.ColorRibbon2:LinearColor=&MakeCustomLinearColorFromBool(
     Bool=$User.bIsPainted
     True Color=&Multiply_LinearColorByFloat(
      Linear Color=<rgba 1 1 1 1>
      Float=&FloatFromCurve(FloatCurve=DataInterface() {propertyValues {Curve {keys [<at 0 1>,<at 1 0>]}}}, Scale Curve=5)
     )
     False Color=&ColorFromCurve(ColorCurve=DataInterface() {propertyValues {
        AlphaCurve {keys [
          {arriveTangent=-1.1807, interpMode=RCIM_Cubic, leaveTangent=-1.1807, tangentMode=RCTM_User, time=0, value=1.25}
          {arriveTangent=-1.084, interpMode=RCIM_Cubic, leaveTangent=-1.084, tangentMode=RCTM_User, time=1, value=0}
         ]}
        BlueCurve {keys [<at 0 5 RCIM_Cubic>,<at 1 5 RCIM_Cubic>]}
        GreenCurve {keys [<at 0 2.0958 RCIM_Cubic>,<at 1 2.0958 RCIM_Cubic>]}
        RedCurve {keys [<at 0 0.75 RCIM_Cubic>,<at 1 0.75 RCIM_Cubic>]}
       }})
    )
    Particles.ColorRibbon3:LinearColor=&MakeCustomLinearColorFromBool(
     Bool=$User.bIsPainted
     True Color=&Multiply_LinearColorByFloat(
      Linear Color=<rgba 1 1 1 1>
      Float=&FloatFromCurve(FloatCurve=DataInterface() {propertyValues {Curve {keys [<at 0 1>,<at 0.3 0.4>,<at 1 0>]}}}, Scale Curve=10)
     )
     False Color=&Multiply_LinearColorByFloat(
      Linear Color=<rgba 4 1.6667 0 0.1>
      Float=&FloatFromCurve(FloatCurve=DataInterface() {propertyValues {Curve {keys [<at 0 1>,<at 1 0>]}}}, Scale Curve=40)
     )
    )
    Particles.ColorRibbon4:LinearColor=&MakeCustomLinearColorFromBool(
     Bool=$User.bIsPainted
     True Color=&Multiply_LinearColorByFloat(
      Linear Color=<rgba 1 1 1 1>
      Float=&FloatFromCurve(FloatCurve=DataInterface() {propertyValues {Curve {keys [<at 0 1>,<at 1 0.5>]}}}, Scale Curve=30)
     )
     False Color=&Multiply_LinearColorByFloat(
      Linear Color=<rgba 0 0.3333 1 0.1>
      Float=&FloatFromCurve(FloatCurve=DataInterface() {propertyValues {Curve {keys [<at 0 1>,<at 1 0.5>]}}}, Scale Curve=100)
     )
    )
    Particles.DMPRibbon2:Vector4f=&MakeVector4(X=0.45, Y=50, Z=1, W=2)
    Particles.DMPRibbon2001:Vector4f=&MakeVector4(
     X=1
     Y=1.5
     Z=&FloatFromCurve(
      FloatCurve=DataInterface() {propertyValues {Curve {keys [{tangentMode=RCTM_User, time=0, value=1}, {tangentMode=RCTM_User, time=0.4, value=4.5}, <at 1 0>]}}}
      Scale Curve=1.5
     )
     W=&FloatFromCurve(
      FloatCurve=DataInterface() {propertyValues {Curve {keys [{tangentMode=RCTM_User, time=0, value=1}, {tangentMode=RCTM_User, time=0.5, value=4}, <at 1 0>]}}}
      Scale Curve=0.6
     )
    )
    Particles.DMPRibbon3:Vector4f=&MakeVector4(
     X=&RandomRangeFloat(Minimum=0.4, Maximum=0.5)
     Y=50
     Z=0
     W=&FloatFromCurve(FloatCurve=DataInterface() {propertyValues {Curve {keys [<at 0 0.75 RCIM_Cubic>,<at 0.8 0 RCIM_Cubic>]}}}, Scale Curve=10)
    )
    Particles.DMPRibbon3001:Vector4f=&MakeVector4(
     X=&FloatFromCurve(FloatCurve=DataInterface() {propertyValues {Curve {keys [{tangentMode=RCTM_User, time=0, value=0}, {tangentMode=RCTM_User, time=0.5, value=1}]}}})
     Y=1.5
     Z=&FloatFromCurve(
      FloatCurve=DataInterface() {propertyValues {Curve {keys [{tangentMode=RCTM_User, time=0, value=0.2}, {tangentMode=RCTM_User, time=0.7, value=1}, <at 1 0>]}}}
      Scale Curve=4
     )
     W=&FloatFromCurve(FloatCurve=DataInterface() {propertyValues {Curve {keys [<at 0 1>,<at 1 0>]}}}, Scale Curve=8)
    )
    Particles.DMPRibbon4:Vector4f=&MakeVector4(
     X=&RandomRangeFloat(Minimum=0.4, Maximum=0.5)
     Y=50
     Z=0
     W=&FloatFromCurve(FloatCurve=DataInterface() {propertyValues {Curve {keys [<at 0 0.75 RCIM_Cubic>,<at 0.8 0 RCIM_Cubic>]}}}, Scale Curve=10)
    )
    Particles.DMPRibbon4001:Vector4f=&MakeVector4(
     X=&FloatFromCurve(FloatCurve=DataInterface() {propertyValues {Curve {keys [{tangentMode=RCTM_User, time=0, value=0.5}, {tangentMode=RCTM_User, time=1, value=1}]}}})
     Y=1
     Z=&FloatFromCurve(
      FloatCurve=DataInterface() {propertyValues {Curve {keys [{tangentMode=RCTM_User, time=0, value=0.2}, {tangentMode=RCTM_User, time=0.5, value=1}]}}}
      Scale Curve=4
     )
     W=&FloatFromCurve(FloatCurve=DataInterface() {propertyValues {Curve {keys [<at 0 1>,<at 1 0>]}}}, Scale Curve=8)
    )
    Particles.RibbonWidth2:NiagaraFloat=&FloatFromCurve(FloatCurve=DataInterface() {propertyValues {Curve {keys [<at 0 1>,<at 1 0>]}}}, Scale Curve=50)
    Particles.RibbonWidth3=5
   ) {enabled=false, isSetParameters=true, module=SetVariables_#2}
   &ScaleRibbonWidth(Ribbon Width Scale=&FloatFromCurve(FloatCurve=DataInterface() {propertyValues {Curve {keys [<at 0 1>,<at 1 0.5>]}}}))
   /SetVariables(Particles.lineswidth:NiagaraFloat=&Multiply_Float(
     A=55
     B=&FloatFromCurve(FloatCurve=DataInterface() {propertyValues {Curve {keys [<at 0 1>,<at 1 0.5>]}}})
    )) {isSetParameters=true}
  ]
  rendererClasses [&NiagaraRibbonRendererProperties]
  renderers [
   {rendererClass=&NiagaraRibbonRendererProperties}
   {rendererClass=&NiagaraRibbonRendererProperties, rendererIndex=1}
   {rendererClass=&NiagaraRibbonRendererProperties, rendererIndex=2}
   {rendererClass=&NiagaraRibbonRendererProperties, rendererIndex=3}
   {rendererClass=&NiagaraRibbonRendererProperties, rendererIndex=4}
  ]
 }
 Ribbons001 {
  bEnabled=false
  emitterUpdateScript [
   &EmitterState(Life Cycle Mode=::NewEnumerator1, Loop Duration=1, Recalculate Duration Each Loop=true, UseLoopDelay=false)
   &VC_FilteredSocketCount()
   &VC_RateQualityHandler(Epic Spawn Rate=30, High Spawn Rate=30, Low Spawn Rate=30, Use Custom Rate?=true)
   &VC_SpawnRibbon_v1(SpawnRate=$Emitter.Spawn Rate)
  ]
  particleSpawnScript [
   &VC_PaintedVariantsOverride() {enabled=false}
   &VC_InitializeRibbon_v1(
    Color=$User.PaintedColor
    Custom Offset=true
    Enable Skeletal Position=true
    Lifetime=0.9
    Local Vector Offset=&AddVectorToPosition(
     Position=&MakeCustomPositionFromBool(Bool:NiagaraBool=$User.bIsFrontendPreview, True Position:NiagaraPosition=<-10 0 0>, False Position:NiagaraPosition=<0 0 0>)
     Vector=$User.DTCarOffset
    )
    Offset Amount=1
    Ribbon Width=60
   )
   &AddVelocity(Velocity=<-2500 0 0>) {module=AddVelocity001}
  ]
  particleUpdateScript [
   &ParticleState()
   &InheritSourceMovement(
    Applied Position Scale=&Multiply_VectorByFloat(
     Vector=<1 1 1>
     Float=&FloatFromCurve(FloatCurve=DataInterface() {propertyValues {Curve {keys [
          {interpMode=RCIM_Cubic, leaveTangent=-0.0051, tangentMode=RCTM_Break, time=0, value=1}
          {arriveTangent=-0.247, interpMode=RCIM_Cubic, tangentMode=RCTM_Break, time=1, value=0.9}
         ]}}})
    )
    ApplyForce=false
    ApplyPosition=true
    ApplyPositionFalloff=false
   )
   &SolveForcesAndVelocity(Clamp Velocity=false, Limit Acceleration=false)
   &VC_UpdateRibbon_v1(Enable Skeletal Position=true)
   &Color(Color=&MakeCustomLinearColorFromBool(
     Bool=$User.bIsPainted
     True Color=$User.PaintedColor
     False Color=&MakeLinearColorFromVectorAndFloat(
      "Vector (RGB)"=&VectorFromCurve(
       VectorCurve=DataInterface() {propertyValues {
         XCurve {keys [<at 0 8>,<at 0.4 3>,<at 1 0>]}
         YCurve {keys [<at 0 8>,<at 0.05 2>,<at 0.5 1>]}
         ZCurve {keys [<at 0 0.1>,<at 0.5 2>,<at 1 8>]}
        }}
       Scale Curve=&VectorFromFloat(Value=0.1)
      )
      "Float (Alpha)"=&FloatFromCurve(FloatCurve=DataInterface() {propertyValues {Curve {keys [<at 0 0 RCIM_Cubic>, {arriveTangent=0.2, interpMode=RCIM_Cubic, leaveTangent=0.2, time=0.05, value=0.9}, <at 1 0.2>]}}})
     )
    ))
   &ScaleColor(
    Scale Alpha=&MakeCustomFloatFromBool(
     Bool=$User.bIsPainted
     True Float=&FloatFromCurve(FloatCurve=DataInterface() {propertyValues {Curve {keys [
          {arriveTangent=15.9245, interpMode=RCIM_Cubic, leaveTangent=15.9245, tangentMode=RCTM_User, time=0, value=0}
          {arriveTangent=1, interpMode=RCIM_Cubic, leaveTangent=1, time=0.1, value=0.9}
          <at 1 1>
         ]}}})
     False Float=&FloatFromCurve(FloatCurve=DataInterface() {propertyValues {Curve {keys [<at 0 0 RCIM_Cubic>, {arriveTangent=0.2, interpMode=RCIM_Cubic, leaveTangent=0.2, time=0.05, value=0.9}, <at 1 0.2>]}}})
    )
    Scale RGB=&MakeCustomPositionFromBool(
     Bool:NiagaraBool=$User.bIsPainted
     True Position:NiagaraPosition=&VectorFromFloat(Value=&FloatFromCurve(FloatCurve=DataInterface() {propertyValues {Curve {keys [<at 0 1 RCIM_Cubic>,<at 0.7 0.05 RCIM_Cubic>]}}}, Scale Curve=2.5))
     False Position:NiagaraPosition=&VectorFromCurve(
      VectorCurve=DataInterface() {propertyValues {
        XCurve {keys [<at 0 8>,<at 0.4 3>,<at 1 0>]}
        YCurve {keys [<at 0 8>,<at 0.05 2>,<at 0.5 1>]}
        ZCurve {keys [<at 0 0.1>,<at 0.5 2>,<at 1 8>]}
       }}
      Scale Curve=&VectorFromFloat(Value=0.1)
     )
    )
    ScaleA=true
    ScaleRGB=true
   ) {module=ScaleColor001}
   &DynamicMaterialParameters(
    Index 0 Param 1=&MakeCustomFloatFromBool(
     Bool=$User.bIsPainted
     True Float=&FloatFromCurve(FloatCurve=DataInterface() {propertyValues {Curve {keys [<at 0 6 RCIM_Cubic>,<at 1 0.75 RCIM_Cubic>]}}}, Scale Curve=$User.GlowOverride)
     False Float=&FloatFromCurve(FloatCurve=DataInterface() {propertyValues {Curve {keys [<at 0 0.5 RCIM_Cubic>,<at 0.15 0.8>,<at 1 1>]}}}, Scale Curve=10)
    )
    Module.Index0Param0Write=true
    Module.Index0Param1Write=true
    Module.Index0Param2Write=true
    Module.Index0Param3Write=true
   )
   /SetVariables(
    Particles.ColorRibbon2:LinearColor=&MakeCustomLinearColorFromBool(
     Bool=$User.bIsPainted
     True Color=&Multiply_LinearColorByFloat(
      Linear Color=<rgba 1 1 1 1>
      Float=&FloatFromCurve(FloatCurve=DataInterface() {propertyValues {Curve {keys [<at 0 1>,<at 1 0>]}}}, Scale Curve=5)
     )
     False Color=&ColorFromCurve(ColorCurve=DataInterface() {propertyValues {
        AlphaCurve {keys [
          {arriveTangent=-1.1807, interpMode=RCIM_Cubic, leaveTangent=-1.1807, tangentMode=RCTM_User, time=0, value=1.25}
          {arriveTangent=-1.084, interpMode=RCIM_Cubic, leaveTangent=-1.084, tangentMode=RCTM_User, time=1, value=0}
         ]}
        BlueCurve {keys [<at 0 5 RCIM_Cubic>,<at 1 5 RCIM_Cubic>]}
        GreenCurve {keys [<at 0 2.0958 RCIM_Cubic>,<at 1 2.0958 RCIM_Cubic>]}
        RedCurve {keys [<at 0 0.75 RCIM_Cubic>,<at 1 0.75 RCIM_Cubic>]}
       }})
    )
    Particles.ColorRibbon3:LinearColor=&MakeCustomLinearColorFromBool(
     Bool=$User.bIsPainted
     True Color=&Multiply_LinearColorByFloat(
      Linear Color=<rgba 1 1 1 1>
      Float=&FloatFromCurve(FloatCurve=DataInterface() {propertyValues {Curve {keys [<at 0 1>,<at 0.3 0.4>,<at 1 0>]}}}, Scale Curve=10)
     )
     False Color=&Multiply_LinearColorByFloat(
      Linear Color=<rgba 4 1.6667 0 0.1>
      Float=&FloatFromCurve(FloatCurve=DataInterface() {propertyValues {Curve {keys [<at 0 1>,<at 1 0>]}}}, Scale Curve=40)
     )
    )
    Particles.ColorRibbon4:LinearColor=&MakeCustomLinearColorFromBool(
     Bool=$User.bIsPainted
     True Color=&Multiply_LinearColorByFloat(
      Linear Color=<rgba 1 1 1 1>
      Float=&FloatFromCurve(FloatCurve=DataInterface() {propertyValues {Curve {keys [<at 0 1>,<at 1 0.5>]}}}, Scale Curve=30)
     )
     False Color=&Multiply_LinearColorByFloat(
      Linear Color=<rgba 0 0.3333 1 0.1>
      Float=&FloatFromCurve(FloatCurve=DataInterface() {propertyValues {Curve {keys [<at 0 1>,<at 1 0.5>]}}}, Scale Curve=100)
     )
    )
    Particles.DMPRibbon2:Vector4f=&MakeVector4(X=0.45, Y=50, Z=1, W=2)
    Particles.DMPRibbon2001:Vector4f=&MakeVector4(
     X=1
     Y=1.5
     Z=&FloatFromCurve(
      FloatCurve=DataInterface() {propertyValues {Curve {keys [{tangentMode=RCTM_User, time=0, value=1}, {tangentMode=RCTM_User, time=0.4, value=4.5}, <at 1 0>]}}}
      Scale Curve=1.5
     )
     W=&FloatFromCurve(
      FloatCurve=DataInterface() {propertyValues {Curve {keys [{tangentMode=RCTM_User, time=0, value=1}, {tangentMode=RCTM_User, time=0.5, value=4}, <at 1 0>]}}}
      Scale Curve=0.6
     )
    )
    Particles.DMPRibbon3:Vector4f=&MakeVector4(
     X=&RandomRangeFloat(Minimum=0.4, Maximum=0.5)
     Y=50
     Z=0
     W=&FloatFromCurve(FloatCurve=DataInterface() {propertyValues {Curve {keys [<at 0 0.75 RCIM_Cubic>,<at 0.8 0 RCIM_Cubic>]}}}, Scale Curve=10)
    )
    Particles.DMPRibbon3001:Vector4f=&MakeVector4(
     X=&FloatFromCurve(FloatCurve=DataInterface() {propertyValues {Curve {keys [{tangentMode=RCTM_User, time=0, value=0}, {tangentMode=RCTM_User, time=0.5, value=1}]}}})
     Y=1.5
     Z=&FloatFromCurve(
      FloatCurve=DataInterface() {propertyValues {Curve {keys [{tangentMode=RCTM_User, time=0, value=0.2}, {tangentMode=RCTM_User, time=0.7, value=1}, <at 1 0>]}}}
      Scale Curve=4
     )
     W=&FloatFromCurve(FloatCurve=DataInterface() {propertyValues {Curve {keys [<at 0 1>,<at 1 0>]}}}, Scale Curve=8)
    )
    Particles.DMPRibbon4:Vector4f=&MakeVector4(
     X=&RandomRangeFloat(Minimum=0.4, Maximum=0.5)
     Y=50
     Z=0
     W=&FloatFromCurve(FloatCurve=DataInterface() {propertyValues {Curve {keys [<at 0 0.75 RCIM_Cubic>,<at 0.8 0 RCIM_Cubic>]}}}, Scale Curve=10)
    )
    Particles.DMPRibbon4001:Vector4f=&MakeVector4(
     X=&FloatFromCurve(FloatCurve=DataInterface() {propertyValues {Curve {keys [{tangentMode=RCTM_User, time=0, value=0.5}, {tangentMode=RCTM_User, time=1, value=1}]}}})
     Y=1
     Z=&FloatFromCurve(
      FloatCurve=DataInterface() {propertyValues {Curve {keys [{tangentMode=RCTM_User, time=0, value=0.2}, {tangentMode=RCTM_User, time=0.5, value=1}]}}}
      Scale Curve=4
     )
     W=&FloatFromCurve(FloatCurve=DataInterface() {propertyValues {Curve {keys [<at 0 1>,<at 1 0>]}}}, Scale Curve=8)
    )
    Particles.RibbonWidth2:NiagaraFloat=&FloatFromCurve(FloatCurve=DataInterface() {propertyValues {Curve {keys [<at 0 1>,<at 1 0>]}}}, Scale Curve=50)
    Particles.RibbonWidth3=5
   ) {isSetParameters=true, module=SetVariables_#2}
   &ScaleRibbonWidth(Ribbon Width Scale=&FloatFromCurve(FloatCurve=DataInterface() {propertyValues {Curve {keys [<at 0 1>,<at 1 0.5>]}}}))
  ]
  rendererClasses [&NiagaraRibbonRendererProperties]
  renderers [
   {rendererClass=&NiagaraRibbonRendererProperties}
   {rendererClass=&NiagaraRibbonRendererProperties, rendererIndex=1}
   {rendererClass=&NiagaraRibbonRendererProperties, rendererIndex=2}
   {rendererClass=&NiagaraRibbonRendererProperties, rendererIndex=3}
  ]
 }
 Sprites {
  emitterUpdateScript [
   &EmitterState()
   &VC_FilteredSocketCount()
   &SpawnRate(SpawnRate=80, Use Spawn Probability=false)
  ]
  particleSpawnScript [
   &VC_PaintedVariantsOverride(DefaultColor=&SelectLinearColorFromArray(
     Array Sampling Mode=::NewEnumerator0
     Direct Array Index=?
     Color Selection Array=DataInterface() {propertyValues {ColorData [<rgba 0.1 0.8917 2 1>, <rgba 2 1.05 0.1 1>]}}
     Random Seed=0
    ))
   &InitializeParticle(
    Color=$Particles.Color
    Lifetime Max=0.8
    Lifetime Min=0.4
    Lifetime Mode=::NewEnumerator1
    Sprite Rotation Mode=::NewEnumerator2
    Sprite Size Mode=::NewEnumerator1
    Uniform Sprite Size Max=13
    Uniform Sprite Size Min=11
    UsePositionOffset=false
    Write Material Random=true
   )
   &PartitionParticles(Number Of Partitions=$Emitter.PortCount)
   &SkeletalMeshLocation(
    Bone / Socket Index=$Particles.PartitionParticles.Partition
    Bone / Socket Sampling Mode=::NewEnumerator5
    MaskSpawnGroup=false
    Mesh Sampling Type=::NewEnumerator6
    Sampled Position Offset Coordinate Space:ENiagaraSkelMeshTransforms=::NewEnumerator2
    Skeletal Mesh=$System.SkeletalMesh
   )
   &ShapeLocation(
    MaskSpawnGroup=false
    Non Uniform Scale=<1 0.5 1>
    Offset=&AddVectorToPosition(Position=<-50 0 0>, Vector=$User.DTCarOffset)
    Offset Coordinate Space=::Local
    Offset Mode=::NewEnumerator0
    Override Seed=false
    Shape Primitive=::NewEnumerator1
    Surface Only=true
    UseOwnerScale=true
   )
   &AddVelocity(Velocity=&RandomRangeVector(Minimum=<-1900 -50 -50>, Maximum=<-2200 50 50>))
  ]
  particleUpdateScript [
   &ParticleState()
   &ScaleSpriteSize(Uniform Curve Sprite Scale=DataInterface() {propertyValues {Curve {keys [
        {arriveTangent=1.3182, interpMode=RCIM_Cubic, leaveTangent=1.3182, leaveTangentWeight=0.3333, tangentMode=RCTM_User, time=0, value=0.8}
        {arriveTangentWeight=0.1617, interpMode=RCIM_Cubic, leaveTangentWeight=0.2652, tangentMode=RCTM_User, time=0.3, value=1}
        {arriveTangent=-3.3914, arriveTangentWeight=0.3333, interpMode=RCIM_Cubic, leaveTangent=-3.3914, tangentMode=RCTM_User, time=1, value=0}
       ]}}})
   &VortexForce(
    Origin Pull Amount=10000
    UseOriginOffset=true
    Vortex Axis=&RandomRangeVector(Minimum=<-1 0 0>, Maximum=<1 0 0>, Evaluation Type=::NewEnumerator1)
    Vortex Axis Coordinate Space:ENiagaraCoordinateSpace=::Local
    Vortex Force Amount=8500
    Vortex Origin:NiagaraPosition=$Particles.Initial.Position
    Vortex Origin Offset:Vector3f=&RandomRangeVector(Minimum=<-1000 0 0>, Maximum=<-400 0 0>)
    Vortex Origin Offset Coordinate Space:ENiagaraCoordinateSpace=::Local
   )
   &InheritSourceMovement(
    Applied Position Scale=&Multiply_VectorByFloat(
     Vector=<1 1 1>
     Float=&FloatFromCurve(FloatCurve=DataInterface() {propertyValues {Curve {keys [{interpMode=RCIM_Cubic, tangentMode=RCTM_User, time=0.3, value=1}, {interpMode=RCIM_Cubic, tangentMode=RCTM_User, time=1, value=0}]}}})
    )
    ApplyForce=false
    ApplyPosition=true
    ApplyPositionFalloff=false
   )
   &CurlNoiseForce(Bias Noise=false, Noise Frequency=1, Noise Strength=50, Pan Noise=false)
   &SolveForcesAndVelocity(Clamp Velocity=false, Limit Acceleration=false)
   &ScaleColor(
    Scale Alpha=&FloatFromCurve(FloatCurve=DataInterface() {propertyValues {Curve {keys [<at 0 0>,<at 0.1 1>,<at 0.8 1>,<at 1 0>]}}})
    Scale RGB=&VectorFromFloat(Value=5)
    ScaleA=true
    ScaleRGB=true
   )
  ]
  rendererClasses [&NiagaraSpriteRendererProperties]
  renderers [{rendererClass=&NiagaraSpriteRendererProperties}]
 }
 Sprites002 {
  bEnabled=false
  emitterUpdateScript [
   &EmitterState()
   &VC_FilteredSocketCount()
   &SpawnRate(
    SpawnRate=&MakeCustomFloatFromBool(Bool=$User.bIsFrontendPreview, True Float=140, False Float=80)
    Use Spawn Probability=false
   )
  ]
  particleSpawnScript [
   &VC_PaintedVariantsOverride() {enabled=false}
   &InitializeParticle(
    Color=&MakeCustomLinearColorFromBool(
     Bool=$User.bIsPainted
     True Color=<rgba 1 1 1 1>
     False Color=&SelectLinearColorFromArray(
      Array Sampling Mode=::NewEnumerator0
      Direct Array Index=?
      Color Selection Array=DataInterface() {propertyValues {ColorData [<rgba 0.5 2.225 5 1>, <rgba 1.5 1.25 0 1>]}}
      Random Seed=0
     )
    )
    Lifetime Max=&MakeCustomFloatFromBool(Bool=$User.bIsFrontendPreview, True Float=0.4, False Float=1)
    Lifetime Min=0.4
    Lifetime Mode=::NewEnumerator1
    Sprite Rotation Mode=::NewEnumerator2
    Sprite Size Mode=::NewEnumerator3
    Uniform Sprite Size=&MakeCustomFloatFromBool(
     Bool=$User.bIsPainted
     True Float=&RandomRangeFloat(Minimum=8, Maximum=10)
     False Float=&RandomRangeFloat(Minimum=10, Maximum=12)
    )
    UsePositionOffset=false
    Write Material Random=true
   )
   &PartitionParticles(Number Of Partitions=$Emitter.PortCount)
   &SkeletalMeshLocation(
    Bone / Socket Index=$Particles.PartitionParticles.Partition
    Bone / Socket Sampling Mode=::NewEnumerator5
    MaskSpawnGroup=false
    Mesh Sampling Type=::NewEnumerator6
    Sampled Position Offset=&AddVectorToPosition(Position=<-50 0 0>, Vector=$User.DTCarOffset)
    Sampled Position Offset Coordinate Space:ENiagaraSkelMeshTransforms=::NewEnumerator2
    Skeletal Mesh=$System.SkeletalMesh
   )
   &AddVelocity(Velocity=&RandomRangeVector(Minimum=<-2000 100 -50>, Maximum=<-3000 -100 50>))
   &ShapeLocation(
    MaskSpawnGroup=false
    Non Uniform Scale=<1 0.5 1>
    Override Seed=false
    Shape Primitive=::NewEnumerator1
    Surface Only=true
    UseOwnerScale=true
   )
  ]
  particleUpdateScript [
   &ParticleState()
   &ScaleSpriteSize(Uniform Curve Sprite Scale=DataInterface() {propertyValues {Curve {keys [
        {interpMode=RCIM_Cubic, leaveTangent=-0.1977, tangentMode=RCTM_Break, time=0, value=1}
        {arriveTangent=-1.1493, interpMode=RCIM_Cubic, leaveTangent=-1.1493, tangentMode=RCTM_User, time=0.5, value=0.85}
        {arriveTangent=-4.2625, interpMode=RCIM_Cubic, tangentMode=RCTM_Break, time=1, value=0}
       ]}}})
   &VortexForce(
    Origin Pull Amount:NiagaraFloat=&MakeCustomFloatFromBool(Bool=$User.bIsFrontendPreview, True Float=10000, False Float=10000)
    UseOriginOffset=true
    Vortex Axis=&RandomRangeVector(Minimum=<-1 0 0>, Maximum=<1 0 0>, Evaluation Type=::NewEnumerator1)
    Vortex Axis Coordinate Space:ENiagaraCoordinateSpace=::Local
    Vortex Force Amount=&MakeCustomFloatFromBool(Bool=$User.bIsFrontendPreview, True Float=10000, False Float=7500)
    Vortex Origin:NiagaraPosition=$Particles.Initial.Position
    Vortex Origin Offset:Vector3f=&RandomRangeVector(Minimum=<-1000 0 0>, Maximum=<-400 0 0>)
    Vortex Origin Offset Coordinate Space:ENiagaraCoordinateSpace=::Local
   )
   &InheritSourceMovement(
    Applied Position Scale=&Multiply_VectorByFloat(
     Vector=<1 1 1>
     Float=&FloatFromCurve(FloatCurve=DataInterface() {propertyValues {Curve {keys [{interpMode=RCIM_Cubic, tangentMode=RCTM_User, time=0.3, value=1}, {interpMode=RCIM_Cubic, tangentMode=RCTM_User, time=1, value=0}]}}})
    )
    ApplyForce=false
    ApplyPosition=true
    ApplyPositionFalloff=false
   )
   &CurlNoiseForce(Bias Noise=false, Noise Frequency=1, Noise Strength=50, Pan Noise=false)
   &SolveForcesAndVelocity(Clamp Velocity=false, Limit Acceleration=false)
   &ScaleColor(
    Scale Alpha=&FloatFromCurve(FloatCurve=DataInterface() {propertyValues {Curve {keys [
         <at 0.02 0>
         <at 0.05 1>
         {interpMode=RCIM_Cubic, tangentMode=RCTM_User, time=0.8, value=1}
         {interpMode=RCIM_Cubic, tangentMode=RCTM_User, time=1, value=0}
        ]}}})
    Scale RGB=&Multiply_VectorByFloat(Vector=<1 1 1>, Float=&MakeCustomFloatFromBool(Bool=$User.bIsPainted, True Float=5, False Float=1))
    ScaleA=true
    ScaleRGB=true
   )
  ]
  rendererClasses [&NiagaraSpriteRendererProperties]
  renderers [{rendererClass=&NiagaraSpriteRendererProperties}]
 }
}
summary {
 systemName=NS_Boost_Lightwave_Backup
 emitters [
  {bEnabled=false, emitterName=Ribbons001, rendererClasses [&NiagaraRibbonRendererProperties]}
  {emitterName=Sprites, rendererClasses [&NiagaraSpriteRendererProperties]}
  {bEnabled=false, emitterName=Sprites002, rendererClasses [&NiagaraSpriteRendererProperties]}
  {emitterName=Ribbons, rendererClasses [&NiagaraRibbonRendererProperties]}
 ]
 userVariables [
  User.bIsFrontendPreview {type=NiagaraBool}
  User.bIsPainted {type=NiagaraBool}
  User.ColorVariantID {type=NiagaraInt32}
  User.DTCarOffset {type=Vector3f, description="", defaultValue {struct=Vector3f, value=<0 0 0>}}
  User.GlowOverride {type=NiagaraFloat, defaultValue=4, description=""}
  User.PaintedColor {type=LinearColor, description="", defaultValue {struct=LinearColor, value=<rgba 1 1 1 1>}}
  User.PaintedScalar {type=NiagaraFloat, defaultValue=1, description=""}
  User.PortCount {type=NiagaraInt32, defaultValue=2}
 ]
}
