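# Unlit translucent material for the beam of the Lightwave boost vehicle cosmetic, built for Niagara ribbons.
# A clamped V-gradient forms a bright band across the ribbon, with the EdgeGlow dynamic parameter setting edge intensity.
# This desaturated mask is tinted by particle color and scaled by particle alpha, giving color and opacity terms.
# A power curve on U fades both terms sharply toward one end of the ribbon.
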Material /VehicleCosmetics/Boosts/Boost_Lightwave/Materials/M_Boost_LightWaveBeam
graph {
 properties {bUsedWithNiagaraRibbons=true, blendMode=BLEND_Translucent, pixelDepthOffsetMode=PDOM_Legacy, shadingModel=MSM_Unlit, usedShadingModels=Unlit}
 nodes [
  MaterialExpressionVertexColor_2 = ParticleColor()
  n1 = ComponentMask(None=TextureCoordinate()) {g=true}
  n2 = OneMinus(None=@n1)
  MaterialExpressionConstantClamp_4 = Saturate(None=Multiply(A=Multiply(A=@n1, B=@n2), B=Constant() {r=12}))
  n3 = Constant() {r=1}
  n4 = OneMinus(None=@MaterialExpressionConstantClamp_4)
  n5 = DynamicParameter() {paramNames [EdgeGlow,Param2,Param3,Param4]}
  n6 = Desaturation(None=Add(A=Multiply(A=@MaterialExpressionConstantClamp_4, B=@n3), B=Multiply(A=@n4, B=@n5.EdgeGlow)))
  n7 = Multiply(A=Multiply(A=Multiply(A=@MaterialExpressionVertexColor_2.A) {constB=2}, B=@n6)) {constB=0.5}
  n8 = Multiply(A=Multiply(A=@MaterialExpressionVertexColor_2.RGB, B=@n6), B=@n7)
  MaterialFunctionCall(Emissive=@n8) {materialFunction=&MF_FX_EmissiveAdjustments}
  n9 = TextureCoordinate()
  n10 = OneMinus(None=Power(Base=ComponentMask(None=@n9) {r=true}) {constExponent=8})
  Multiply(A=@n7, B=@n10)
  Multiply(A=@n8, B=@n10)
 ]
}
