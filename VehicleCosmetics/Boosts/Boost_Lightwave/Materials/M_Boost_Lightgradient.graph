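# Translucent unlit Niagara sprite material for the Lightwave vehicle boost effect, drawing a soft round glow.
# It builds a radial gradient procedurally, without textures, from squared UV distance to the sprite center, remapped and saturated to fade out before the edges.
# Particle color RGB times the gradient feeds MF_FX_EmissiveAdjustments, while particle alpha times the gradient goes through a Power node for a tighter falloff.
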Material /VehicleCosmetics/Boosts/Boost_Lightwave/Materials/M_Boost_Lightgradient
graph {
 properties {bUsedWithNiagaraSprites=true, blendMode=BLEND_Translucent, pixelDepthOffsetMode=PDOM_Legacy, shadingModel=MSM_Unlit, usedShadingModels=Unlit}
 nodes [
  MaterialExpressionVertexColor_6 = ParticleColor()
  n1 = Add(A=TextureCoordinate(), B=Constant() {r=-0.5})
  n2 = OneMinus(None=DotProduct(A=@n1, B=@n1))
  n3 = Constant() {r=0.8}
  MaterialExpressionConstantClamp_10 = Saturate(None=Multiply(A=Subtract(A=@n2, B=@n3), B=Constant() {r=5}))
  MaterialFunctionCall(Emissive=Multiply(A=@MaterialExpressionVertexColor_6.RGB, B=@MaterialExpressionConstantClamp_10)) {materialFunction=&MF_FX_EmissiveAdjustments}
  Power(Base=Multiply(A=@MaterialExpressionVertexColor_6.A, B=@MaterialExpressionConstantClamp_10))
 ]
}
