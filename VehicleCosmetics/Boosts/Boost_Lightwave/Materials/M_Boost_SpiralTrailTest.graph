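# Translucent unlit test material for Niagara ribbon and mesh-particle spiral trails in the Lightwave boost.
# Layered sines and cosines of scrolling UVs build a spiral-like vector, apparently for vertex displacement, controlled by one dynamic parameter set.
# STACKED and STRINGS switches select textured dots, a soft strand or procedural dots, faded along U.
# A second parameter set drives rainbow tint and mask strength; no material output connections appear.
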Material /VehicleCosmetics/Boosts/Boost_Lightwave/Materials/M_Boost_SpiralTrailTest
graph {
 properties {
  bUsedWithNiagaraMeshParticles=true
  bUsedWithNiagaraRibbons=true
  blendMode=BLEND_Translucent
  pixelDepthOffsetMode=PDOM_Legacy
  shadingModel=MSM_Unlit
  twoSided=true
  usedShadingModels=Unlit
 }
 nodes [
  n1 = DynamicParameter() {parameterIndex=1}
  n2 = Time()
  n3 = Multiply(A=@n1.Param2, B=OneMinus(None=Multiply(A=@n2) {constB=5}))
  n4 = ComponentMask(None=Multiply(A=@n1.Param1, B=TextureCoordinate() {uTiling=3})) {r=true}
  Add(A=Multiply(A=@n3, B=Constant() {r=0.5}), B=@n4)
  n5 = ParticleColor()
  MaterialFunctionCall(Emissive=@n5.RGB) {materialFunction=&MF_FX_EmissiveAdjustments}
  n6 = Add(A=@n3, B=@n4)
  n7 = Add(A=Add(A=Cosine(None=@n6), B=Cosine(None=@n6) {period=2}), B=Cosine(None=@n6) {period=0.75})
  n8 = Add(A=@n4, B=@n3)
  n9 = Add(A=Add(A=Sine(None=@n8), B=Sine(None=@n8) {period=2}), B=Sine(None=@n8) {period=0.75})
  n10 = Multiply(A=AppendVector(A=@n7, B=AppendVector(A=@n7, B=@n9)), B=@n1.Param3)
  Multiply(
   A=Time() {bOverride_Period=true, period=1}
   B=Transform(None=Multiply(A=@n10, B=@n1.Param4)) {transformSourceType=TRANSFORMSOURCE_World, transformType=TRANSFORM_Local}
  )
  n11 = Multiply(A=TextureCoordinate() {uTiling=50}, B=Constant2Vector() {g=1, r=2}) {constB=1.5}
  n12 = ComponentMask(None=TextureCoordinate()) {g=true}
  n13 = Multiply(A=@n12, B=OneMinus(None=@n12))
  n14 = Constant()
  MaterialExpressionConstantClamp_8 = Saturate(None=Multiply(A=Subtract(A=@n13, B=@n14), B=Constant() {r=2}))
  n15 = OneMinus(None=ComponentMask(None=TextureCoordinate()) {r=true})
  n16 = Multiply(A=@n15, B=@MaterialExpressionConstantClamp_8)
  n17 = Add(A=Frac(None=TextureCoordinate() {uTiling=48}), B=Constant() {r=-0.5})
  n18 = OneMinus(None=DotProduct(A=@n17, B=@n17))
  n19 = Constant() {r=0.95}
  MaterialExpressionConstantClamp_7 = Saturate(None=Multiply(A=Subtract(A=@n18, B=@n19), B=Constant() {r=12}))
  n20 = Multiply(
   A=@n15
   B=StaticSwitchParameter(
    True=TextureSample(UVs=@n11) {texture=/VehicleCosmetics/Boosts/Boost_Lightwave/Textures/T_VCVFX_Boost_Lightwave_Dots}.R
    False=StaticSwitchParameter(True=@n16, False=@MaterialExpressionConstantClamp_7) {parameterName=STRINGS}
   ) {parameterName=STACKED}
  )
  Multiply(A=@n5.A, B=@n20)
  n21 = DynamicParameter()
  n22 = Add(A=TextureCoordinate() {vTiling=0}, B=@n21.Param1)
  Multiply(
   A=@n21.Param2
   B=Desaturation(None=TextureSample(UVs=@n22) {texture=/DealershipSharedLibrary/FX/Textures/Diffuse/T_VCVFX_Rainbow_D}.RGB, Fraction=@n21.Param3)
  )
  Multiply(A=@n21.Param4, B=@n20)
  MaterialExpressionVertexColor_4 = ParticleColor()
 ]
}
